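# Translucent, unlit, two-sided effect material for Niagara mesh particles in the TrustFan emote.
# Two offset gradient samples, one rotated, multiply to drive a three-color lerp through an emissive adjustment function.
# Opacity combines a noise-distorted square mask, vertical edge falloff, smoothstepped Fresnel and particle alpha.
# Depth fade, camera distance fade and a hard scene-depth threshold further mask it, hiding pixels where scene geometry is near the camera.
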Material /BRCosmetics/Animation/Game/MainPlayer/Emotes/TrustFan/FX/M_Emote_TrustFan_Cyc
graph {
 properties {
  bAutomaticallySetUsageInEditor=false
  bUsedWithNiagaraMeshParticles=true
  blendMode=BLEND_Translucent
  shadingModel=MSM_Unlit
  translucencyLightingMode=TLM_SurfacePerPixelLighting
  twoSided=true
  usedShadingModels=Unlit
 }
 nodes [
  Constant() {r=1}
  n1 = TextureSample(UVs=Add(A=TextureCoordinate() {uTiling=1.6}, B=ScalarParameter() {defaultValue=-0.25, parameterName=Offset})) {
   samplerSource=SSM_Clamp_WorldGroupSettings
   texture=/BRCosmetics/Characters/Player/Female/Medium/Bodies/F_MED_CowboyHat/Skins/Steer/Textures/BlackWhite_Gradient
  }
  n2 = TextureCoordinate() {vTiling=0.8}
  n3 = Constant() {r=0.25}
  n4 = VectorParameter() {parameterName=ColorOffset, defaultValue=<rgba 0.05 -0.1 0 1>}
  n5 = TextureSample(UVs=Add(
    A=MaterialFunctionCall(UVs=@n2, "Rotation Angle (0-1)"=@n3) {materialFunction=&CustomRotator}.Rotated Values
    B=AppendVector(A=@n4.R, B=@n4.R)
   )) {
   samplerSource=SSM_Clamp_WorldGroupSettings
   texture=&BlackWhite_Gradient
  }
  MaterialFunctionCall(Emissive=MaterialFunctionCall(
    A=VectorParameter() {parameterName=ColorDarkest, defaultValue=<rgba 0 0 0 1>}.RGB
    B=VectorParameter() {parameterName=ColorDark, defaultValue=<rgba 1 0.0092 0 1>}.RGB
    C=VectorParameter() {parameterName=ColorLight, defaultValue=<rgba 0.0608 0 1 1>}.RGB
    Alpha=Multiply(A=@n1.R, B=@n5.R)
   ) {materialFunction=&Lerp_3Color}.Result) {materialFunction=&MF_FX_EmissiveAdjustments}
  n6 = DepthFade(FadeDistance=ScalarParameter() {parameterName=DepthFade})
  n7 = TextureCoordinate()
  n8 = VectorParameter() {parameterName=SquareMaskUVs, defaultValue=<rgba 0 0.1 1 0.7>}
  n9 = AppendVector(A=@n8.B, B=@n8.B)
  n10 = VectorParameter() {parameterName=SquareMaskDistortionPan, defaultValue=<rgba 0.1 0.1 0 1>}
  n11 = Panner(Speed=AppendVector(A=@n10.R, B=@n10.R))
  n12 = Add(
   A=Add(A=Multiply(A=@n7, B=@n9), B=AppendVector(A=@n8.R, B=@n8.R))
   B=Multiply(
    A=TextureSample(UVs=@n11) {texture=/Game/Global/Textures/Default/Noise/T_SoftNoise}.R
    B=ScalarParameter() {parameterName=DistortionStrength}
   )
  )
  n13 = TextureCoordinate()
  n14 = ComponentMask(None=@n13) {g=true}
  n15 = Multiply(A=@n14, B=OneMinus(None=@n14))
  n16 = Clamp(None=Multiply(A=@n15) {constB=3})
  n17 = SmoothStep(
   Min=ScalarParameter() {defaultValue=0.9, parameterName=FresnelSSMin}
   Max=ScalarParameter() {parameterName=FresnelSSMax}
   Value=Fresnel(ExponentIn=ScalarParameter() {defaultValue=0.2, parameterName=Fresnel}) {exponent=0.45}
  ) {constMax=0, constMin=1}
  n18 = Multiply(
   A=Multiply(
    A=TextureSample(UVs=@n12) {samplerSource=SSM_Clamp_WorldGroupSettings, texture=/BRCosmetics/Effects/Fort_Effects/Emotes/LemonWrath/T_Emote_LemonWrath_Square}.R
    B=Multiply(A=@n16, B=@n17)
   )
   B=ParticleColor().A
  )
  n19 = ScalarParameter() {parameterName=CameraDistanceFade}
  n20 = ScreenPosition()
  n21 = Constant()
  Multiply(
   A=Multiply(
    A=Multiply(A=@n6, B=@n18)
    B=MaterialFunctionCall!(Distance=@n19) {materialFunction=/Game/Effects/Fort_Effects/Materials/_Functions/MF_CameraDistanceFade}.Result
   )
   B=If(
    A=SceneTexture(UVs=@n20.ViewportUV) {sceneTextureId=PPI_SceneDepth}.Color
    B=ScalarParameter() {defaultValue=280, parameterName=SceneDepthTreshold}
    "A > B"=Constant() {r=1}
    "A == B"=@n21
    "A < B"=@n21
   ) {constB=0.01}
  )
  n22 = Add(A=@n13) {constB=0.2}
  n23 = Multiply(A=OneMinus(None=ComponentMask(None=@n22) {r=true}), B=ComponentMask(None=Add(A=@n13) {constB=-0.2}) {r=true})
  Power(Base=SmoothStep(Value=Multiply(A=@n15, B=@n23)) {constMax=0.85}) {constExponent=0.8}
 ]
}
