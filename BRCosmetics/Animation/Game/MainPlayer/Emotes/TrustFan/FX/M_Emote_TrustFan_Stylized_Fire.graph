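# Translucent unlit Niagara sprite material for stylized fire in the TrustFan emote.
# A SubUV texture mask is eroded by panning swirly-smoke noise over per-cell tiled UVs, with dynamic parameters driving erosion and top fade.
# Particle color blends toward a secondary dynamic-parameter value along a gradient, then toward DetailColor through a scrolling flake mask.
# Opacity adds depth fade and optional view-angle fade, both skipped on mobile, plus optional camera-distance fade.
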Material /BRCosmetics/Animation/Game/MainPlayer/Emotes/TrustFan/FX/M_Emote_TrustFan_Stylized_Fire
graph {
 properties {
  bAutomaticallySetUsageInEditor=false
  bUsedWithNiagaraSprites=true
  blendMode=BLEND_Translucent
  shadingModel=MSM_Unlit
  translucencyLightingMode=TLM_SurfacePerPixelLighting
  twoSided=true
  usedShadingModels=Unlit
 }
 nodes [
  n1 = TextureCoordinate()
  Add(A=@n1, B=ParticleRandom())
  n2 = ParticleColor()
  n3 = DynamicParameter() {parameterIndex=1, defaultValue=<rgba 0 1 1 1>, paramNames [Secondary R, Secondary G, Secondary B, Secondary Offset]}
  n4 = MaterialFunctionCall(R=@n3.Secondary R, G=@n3.Secondary R, B=@n3.Secondary R) {materialFunction=&AppendMany}
  n5 = NamedRerouteUsage()
  n6 = Clamp(None=ComponentMask(None=Add(A=@n5, B=@n3.Secondary Offset)) {g=true})
  MaterialFunctionCall(Emissive=LinearInterpolate(
    A=LinearInterpolate(A=@n2.RGB, B=@n4.RGB, Alpha=@n6)
    B=VectorParameter() {parameterName=DetailColor, defaultValue=<rgba 0.1 0.0134 0.0067 1>}.RGB
    Alpha=NamedRerouteUsage()
   ) {constB=0}) {materialFunction=&MF_FX_EmissiveAdjustments}
  n7 = TextureSampleParameter2D() {parameterName=SubUV Texture, texture=/Game/Weapons/FORT_Melee/Effects/Spear_BoomBox/T_FX_MusicNotes_01}
  n8 = VectorParameter() {parameterName=Erosion SmoothMinMax, defaultValue=<rgba 0 0.9 0 0>}
  n9 = ScalarParameter() {defaultValue=5, parameterName=U Count}
  n10 = ScalarParameter() {defaultValue=2, parameterName=V Count}
  n11 = Frac(None=Multiply(A=@n1, B=AppendVector(A=@n9, B=@n10)) {constB=2})
  n12 = Constant() {r=0.25}
  n13 = TextureSampleParameter2D(UVs=Panner(Coordinate=MaterialFunctionCall(UVs=@n11, "Rotation Angle (0-1)"=@n12) {materialFunction=&CustomRotator}.Rotated Values) {speedX=-0.5}) {parameterName=Erosion Texture, texture=/Game/Effects/Fort_Effects/Textures/T_Swirly_Smoke}
  n14 = Add(A=NamedRerouteUsage(), B=Multiply(A=SmoothStep(Min=@n8.R, Max=@n8.R, Value=@n13.R) {constMax=0.9}) {constB=0.4})
  n15 = DynamicParameter() {defaultValue=<rgba 0 1 1 1>, paramNames [Erosion,TopFadeMask,Param3,Param4]}
  n16 = Constant()
  n17 = Constant() {r=1}
  n18 = Constant() {r=-1}
  n19 = Constant() {r=1}
  n20 = ComponentMask(None=Add(
    A=@n14
    B=MaterialFunctionCall(Input=@n15.Erosion, Input Low=@n16, Input High=@n17, Target Low=@n18, Target High=@n19) {materialFunction=&RemapValueRange}.Result
   )) {g=true}
  n21 = Add(A=@n14, B=@n15.TopFadeMask)
  n22 = Multiply(A=Clamp(None=OneMinus(None=@n20)), B=Clamp(None=ComponentMask(None=@n21) {g=true}))
  n23 = DepthFade(Opacity=@n2.A, FadeDistance=ScalarParameter() {defaultValue=50, group=Opacity, parameterName=DepthFade})
  n24 = ScalarParameter() {group=Opacity, parameterName=Angle Fade Bias}
  n25 = LinearInterpolate(
   A=@n23
   B=Multiply(
    A=@n23
    B=MaterialFunctionCall("FadeSensitivity (optional)"=@n24) {materialFunction=&FadeBasedOnViewAngle}.Result
   )
   Alpha=ScalarParameter() {group=Opacity, parameterName=Use Angle Fade}
  )
  n26 = ScalarParameter() {defaultValue=5000, parameterName=DistanceFade}
  Multiply(
   A=Multiply(A=Multiply(A=@n7.R, B=@n22), B=ShadingPathSwitch(Default=@n25, Mobile=@n2.A))
   B=LinearInterpolate(
    B=MaterialFunctionCall!(Distance=@n26) {materialFunction=/Game/Effects/Fort_Effects/Materials/_Functions/MF_CameraDistanceFade}.Result
    Alpha=ScalarParameter() {parameterName=UseDistanceFade}
   ) {constA=1}
  )
  NamedRerouteDeclaration(None=@n11) {name=Frac UVs, nodeColor=<rgba 0 0.4141 1 1>}
  n27 = NamedRerouteUsage()
  n28 = ParticleRandom()
  n29 = Panner(Coordinate=Multiply(A=Add(A=@n27, B=@n28), B=Constant2Vector() {g=2, r=1})) {speedY=0.2}
  NamedRerouteDeclaration(None=LinearInterpolate(A=TextureSample(UVs=@n29) {texture=/Game/Global/Textures/Default/PointSplatter/BoatFlake}.R, Alpha=SmoothStep(Value=@n13.R) {constMax=0.2}) {constB=0}) {name=DetailMask, nodeColor=<rgba 0 1 0.0625 1>}
  ScalarParameter() {parameterName=Erosion}
  ScalarParameter() {defaultValue=1, parameterName=TopFadeMask}
 ]
}
