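# Niagara system for the flames of the TrustFan emote: layered flipbook fire, embers, glows and lights around the character.
# Burst-spawned fire sprites sit on ring arcs, with execution-index curves shaping size and color across each arc and dynamic material parameters animated over life.
# BigFire fades alpha by camera angle, and the Cyc mesh emitter derives a depth threshold from camera distance.
# Curl-noise embers add motion, while Glows adjusts placement by platform.
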NiagaraSystem /BRCosmetics/Animation/Game/MainPlayer/Emotes/TrustFan/FX/NS_Emote_TrustFan_Flames
dependencies {
 usedDataInterfaces [
  &NiagaraDataInterfaceCurve
  &NiagaraDataInterfaceArrayColor
  &NiagaraDataInterfaceArrayFloat2
  &NiagaraDataInterfaceSpriteRendererInfo
  &NiagaraDataInterfaceCamera
  &NiagaraDataInterfacePlatformSet
  &NiagaraDataInterfaceArrayFloat3
  &NiagaraDataInterfaceArrayFloat
  &NiagaraDataInterfaceArrayInt32
  &NiagaraDataInterfaceColorCurve
  &NiagaraDataInterfaceSkeletalMesh
 ]
 usedDynamicInputs [&Vector2DFromFloat, &UniformRangedFloat]
 usedModules [
  &SystemState
  &self:SystemSpawnScript.SystemScriptSource.SystemScriptGraph.NiagaraNodeAssignment_0.SetVariables_#1
 ]
}
emitters {
 BackgroundFire {
  emitterUpdateScript [
   &EmitterState(Life Cycle Mode=::NewEnumerator1, Loop Behavior=::NewEnumerator1, Loop Duration=4, UseLoopDelay=false)
   &SpawnBurst_Instantaneous(Spawn Count=3, Use Spawn Probability=false, UseLoopCountLimit=false)
   &SpawnRate(
    SpawnRate=&FloatFromCurve(FloatCurve=DataInterface() {propertyValues {Curve {keys [<at 0.648 1>,<at 1 0>]}}}, CurveIndex=$Emitter.NormalizedLoopAge, Scale Curve=15)
    Use Spawn Probability=false
   ) {enabled=false}
  ]
  particleSpawnScript [
   /SetVariables(Particles.VisibilityTag=&RandomRangeInt(Minimum=0, Maximum=1)) {isSetParameters=true, module=SetVariables_#2}
   /SetVariables(Particles.Color Array=DataInterface() {propertyValues {GpuSyncMode=SyncCpuToGpu, MaxElements=0, ColorData [<rgba 1 0 0.0007 1>, <rgba 1 0.0896 0 1>, <rgba 0 0 0 0>]}}) {isSetParameters=true, module=SetVariables_#3}
   /SetVariables(
    Particles.Placement:NiagaraFloat=&RandomRangeFloat(Minimum=0, Maximum=1)
    Particles.Scale:NiagaraFloat=&RandomRangeFloat(Minimum=0.9, Maximum=0.9)
   ) {isSetParameters=true, module=SetVariables_#4}
   &InitializeParticle(
    Color=&SelectLinearColorFromArray(
     Array Sampling Mode=::NewEnumerator0
     Direct Array Index=?
     Color Selection Array=DataInterface() {propertyValues {ColorData [<rgba 1 0 0.0007 1>, <rgba 1 0.0156 0 1>]}}
     Random Seed=0
    )
    Lifetime=2.3
    Sprite Size=&Multiply_Vector2D(
     A=&Lerp_Vector2(A=<35 59.2308>, B=<39 71.0769>, Alpha=$Particles.Scale)
     B=&SelectVector2DFromArray(
      Array Sampling Mode=::NewEnumerator0
      Direct Array Index=?
      Vector 2D Selection Array=DataInterface() {propertyValues {FloatData [<-1 1>, <1 1>]}}
      Random Seed=0
     )
    )
    Sprite Size Mode=::NewEnumerator4
    UsePositionOffset=true
    Write Material Random=true
   )
   &SpriteFacingAndAlignment(Sprite Alignment=<0 0 1>, WriteAlignment=true, WriteFacing=true)
   /SetVariables(Particles.Color=&MultiplyLinearColor(
     Linear Color=$Particles.Color
     Scale Factor=&MakeVector4FromVectorAndFloat(
      "Vector (RGB)"=&VectorFromFloat(Value=&FloatFromCurve(
        FloatCurve=DataInterface() {propertyValues {Curve {keys [
            <at 0 0.2 RCIM_Cubic>
            {arriveTangent=3.4824, interpMode=RCIM_Cubic, leaveTangent=-2.7686, tangentMode=RCTM_Break, time=0.502, value=0.0965}
            <at 1 0.2 RCIM_Cubic>
           ]}}}
        CurveIndex=&ReturnNormalizedExecIndex()
        Scale Curve=-1
       ))
      "Float (Alpha)"=&Lerp_Float(
       A=0.1
       B=0.15
       Alpha=&FloatFromCurve(
        FloatCurve=DataInterface() {propertyValues {Curve {keys [{arriveTangent=3.4824, interpMode=RCIM_Cubic, leaveTangent=-3.548, tangentMode=RCTM_Break, time=0.502, value=1}]}}}
        CurveIndex=&ReturnNormalizedExecIndex()
        Scale Curve=7
       )
      )
     )
    )) {isSetParameters=true, module=SetVariables_#5}
  ]
  particleUpdateScript [
   &ParticleState()
   &ScaleSpriteSize(
    Scale Factor=&Multiply_Vector2D_ByFloat(
     Float=&FloatFromCurve(
      FloatCurve=DataInterface() {propertyValues {Curve {keys [{arriveTangent=-3.0981, interpMode=RCIM_Cubic, leaveTangent=-3.0981, tangentMode=RCTM_User, time=0, value=1.5}, <at 0.4322 1 RCIM_Cubic>]}}}
      CurveIndex=$Emitter.Age
     )
     Vector2D=&MakeVector2D(
      X=&FloatFromCurve(
       FloatCurve=DataInterface() {propertyValues {Curve {
          defaultValue=1
          keys [
           {arriveTangent=9.0725, interpMode=RCIM_Cubic, leaveTangent=9.0725, tangentMode=RCTM_User, time=0, value=0}
           {arriveTangent=-2.5249, interpMode=RCIM_Cubic, leaveTangent=-2.5249, tangentMode=RCTM_User, time=0.0263, value=1}
          ]
         }}}
       Scale Curve=&FloatFromCurve(
        FloatCurve=DataInterface() {propertyValues {Curve {
           defaultValue=1
           keys [
            {interpMode=RCIM_Cubic, tangentMode=RCTM_User, time=0, value=0.65}
            {arriveTangent=1.1863, interpMode=RCIM_Cubic, leaveTangent=-1.2112, tangentMode=RCTM_Break, time=0.5, value=1.2}
            {tangentMode=RCTM_User, time=1, value=0.65}
           ]
          }}}
        CurveIndex=&ReturnNormalizedExecIndex()
        Scale Curve=&RandomRangeFloat(Minimum=1.4, Maximum=1.4)
       )
      )
      Y=&FloatFromCurve(
       FloatCurve=DataInterface() {propertyValues {Curve {
          defaultValue=1
          keys [
           {arriveTangent=9.0725, interpMode=RCIM_Cubic, time=0, value=0}
           {arriveTangent=17.0555, interpMode=RCIM_Cubic, leaveTangent=17.0555, tangentMode=RCTM_User, time=0.0262, value=0.9401}
           {arriveTangent=0.124, interpMode=RCIM_Cubic, leaveTangent=0.124, tangentMode=RCTM_User, time=0.0482, value=1}
           {arriveTangent=0.3152, interpMode=RCIM_Cubic, leaveTangent=0.3152, time=0.8993, value=1}
           {arriveTangent=-0.0405, interpMode=RCIM_Cubic, leaveTangent=-0.0405, tangentMode=RCTM_User, time=1, value=1.3}
          ]
         }}}
       Scale Curve=&FloatFromCurve(
        FloatCurve=DataInterface() {propertyValues {Curve {
           defaultValue=1
           keys [
            {interpMode=RCIM_Cubic, tangentMode=RCTM_User, time=0, value=0.65}
            {arriveTangent=0.8018, interpMode=RCIM_Cubic, leaveTangent=-0.6589, tangentMode=RCTM_Break, time=0.5, value=0.8229}
            {tangentMode=RCTM_User, time=1, value=0.65}
           ]
          }}}
        CurveIndex=&ReturnNormalizedExecIndex()
        Scale Curve=&RandomRangeFloat(Minimum=3.5, Maximum=4.4)
       )
      )
     )
    )
    Scale Sprite Size Mode=::NewEnumerator2
   )
   &SubUVAnimation(
    "Sprite Renderer "=DataInterface() {propertyValues {SpriteRenderer=&self:BackgroundFire_0.NiagaraSpriteRendererProperties_0}}
    Start Frame Offset:NiagaraInt32=&RandomRangeInt(Minimum=0, Maximum=10)
    SubUV Lookup Index:NiagaraFloat=&Multiply_Float(
     A=$Particles.Age
     B=&FloatFromCurve(
      FloatCurve=DataInterface() {propertyValues {Curve {keys [<at 0.2594 1.5>,<at 1 1.3>]}}}
      CurveIndex=$Emitter.Age
      Scale Curve=&RandomRangeFloat(Minimum=0.6, Maximum=0.9)
     )
    )
    UseEndFrame=false
    UseStartFrame=false
   )
   &ShapeLocation(
    MaskSpawnGroup=false
    Offset=<-10 -15 -5>
    Offset Mode=::NewEnumerator0
    Override Seed=false
    Ring / Disc Distribution Mode:ENiagaraSphereDistributionMode=::NewEnumerator1
    Ring Radius=&RandomRangeFloat(Minimum=70, Maximum=72)
    Rotation Angle=0.04
    Rotation Axis=<0 0 1>
    Rotation Mode=::NewEnumerator1
    Shape Primitive=::NewEnumerator4
    U Position=&SelectFloatFromArray(
     Direct Array Index=&ReturnExecIndex()
     Float Selection Array=DataInterface() {propertyValues {FloatData [0.54,0.65,-0.13]}}
    )
    UseOwnerScale=true
   ) {module=ShapeLocation001}
   &ScaleColor(
    Scale Alpha=&FloatFromCurve(FloatCurve=DataInterface() {propertyValues {Curve {keys [<at 0.4 1>,<at 1 1>]}}})
    Scale RGB=&VectorFromFloat(Value=&FloatFromCurve(FloatCurve=DataInterface() {propertyValues {Curve {keys [<at 0.3 1>,<at 0.8 1>]}}}))
    ScaleA=true
    ScaleRGB=true
   )
   &DynamicMaterialParameters(
    Index 0 Param 1=&Add_Float(
     A=&FloatFromCurve(FloatCurve=DataInterface() {propertyValues {Curve {keys [<at 0.8166 -1>,<at 1 1>]}}})
     B=&FloatFromCurve(
      FloatCurve=DataInterface() {propertyValues {Curve {keys [<at 1.2 0>,<at 1.5 0>]}}}
      CurveIndex=$Emitter.Age
      Scale Curve=&FloatFromCurve(
       FloatCurve=DataInterface() {propertyValues {Curve {
          defaultValue=0
          keys [
           <at 0 0 RCIM_Cubic>
           {interpMode=RCIM_Cubic, tangentMode=RCTM_User, time=0.9658, value=0}
           {arriveTangent=1.204, interpMode=RCIM_Cubic, leaveTangent=1.204, tangentMode=RCTM_User, time=1, value=1}
          ]
         }}}
       Scale Curve=&RandomRangeFloat(Minimum=1, Maximum=1)
      )
     )
    )
    Index 0 Param 2=10
    Index 1 Param 4=-0.3946
    Index 1 Vector 1:Vector3f=&MakeVectorFromLinearColorRGB(Color=<rgba 0.7812 0.0019 0 1>)
    Module.Index0Float0Write=true
    Module.Index0Float1Write=true
    Module.Index0Float2Write=true
    Module.Index0Float3Write=true
    Module.Index1Float3Write=true
    Module.Index1VectorWrite=true
    Write Parameter Index 1=::NewEnumerator3
   )
   /SetVariables(
    Particles.ScorchAlignment:Vector3f=<0 1 0>
    Particles.ScorchCol:LinearColor=&MakeLinearColorFromVectorAndFloat(
     "Vector (RGB)"=&MakeVectorFromLinearColorRGB(Color=&MultiplyLinearColor(
       Linear Color=$Particles.Initial.Color
       Scale Factor=&MakeVector4_FromLinearColor(LinearColor=<rgba 8 8 8 1>)
      ))
     "Float (Alpha)"=&FloatFromCurve(FloatCurve=DataInterface() {propertyValues {Curve {
         defaultValue=1
         keys [
          {interpMode=RCIM_Cubic, tangentMode=RCTM_User, time=0, value=0}
          {interpMode=RCIM_Cubic, tangentMode=RCTM_User, time=0.0148, value=0.9958}
          {arriveTangent=-1.0108, interpMode=RCIM_Cubic, leaveTangent=-1.0108, time=0.7691, value=0.8518}
          {tangentMode=RCTM_User, time=1, value=0}
         ]
        }}})
    )
    Particles.ScorchFacing:Vector3f=<0 0 1>
    Particles.ScorchPage:Vector4f=&MakeVector4(
     X=&FloatFromCurve(
      FloatCurve=DataInterface() {propertyValues {Curve {defaultValue=1, keys [{tangentMode=RCTM_User, time=0, value=0}, {tangentMode=RCTM_User, time=1, value=1}]}}}
      Scale Curve=10
     )
     Y=0
     Z=0
     W=0
    )
    Particles.ScorchSize:Vector2f=&Multiply_Vector2D_ByFloat(Float=&RandomRangeFloat(Minimum=1, Maximum=1), Vector2D=<150 150>)
   ) {isSetParameters=true, module=SetVariables_#6}
   &OffsetPosition(Position Offset=<0 0 2.7>, Position Output=1, Position to Offset:NiagaraPosition=$Particles.ShapeLocation001.ShapePosition)
   &OffsetPosition(
    Position Offset=&RandomRangeVector(Minimum=<0 0 2.69>, Maximum=<0 0 2.71>)
    Position Output=1
    Position to Offset:NiagaraPosition=$Particles.ShapeLocation001.ShapePosition
   ) {module=OffsetPosition001}
  ]
  rendererClasses [&NiagaraSpriteRendererProperties]
  renderers [
   {rendererClass=&NiagaraSpriteRendererProperties}
   {rendererClass=&NiagaraSpriteRendererProperties, rendererIndex=1}
   {rendererClass=&NiagaraSpriteRendererProperties, rendererIndex=2}
   {rendererClass=&NiagaraSpriteRendererProperties, rendererIndex=3}
   {rendererClass=&NiagaraSpriteRendererProperties, rendererIndex=4}
   {rendererClass=&NiagaraSpriteRendererProperties, rendererIndex=5}
  ]
 }
 BigFire {
  emitterUpdateScript [
   &EmitterState(Life Cycle Mode=::NewEnumerator1, Loop Duration=4, UseLoopDelay=false)
   &SpawnBurst_Instantaneous(Spawn Count=6, Use Spawn Probability=false, UseLoopCountLimit=false)
   &SpawnRate(
    SpawnRate=&FloatFromCurve(FloatCurve=DataInterface() {propertyValues {Curve {keys [<at 0.648 1>,<at 1 0>]}}}, CurveIndex=$Emitter.NormalizedLoopAge, Scale Curve=15)
    Use Spawn Probability=false
   ) {enabled=false}
  ]
  particleSpawnScript [
   /SetVariables(Particles.VisibilityTag=&RandomRangeInt(Minimum=1, Maximum=0)) {isSetParameters=true, module=SetVariables_#2}
   /SetVariables(Particles.Color Array=DataInterface() {propertyValues {GpuSyncMode=SyncCpuToGpu, MaxElements=0, ColorData [<rgba 1 0 0.0007 1>, <rgba 1 0.0896 0 1>, <rgba 0 0 0 0>]}}) {isSetParameters=true, module=SetVariables_#3}
   /SetVariables(
    Particles.Placement:NiagaraFloat=&RandomRangeFloat(Minimum=0, Maximum=1)
    Particles.Scale:NiagaraFloat=&RandomRangeFloat(Minimum=0.9, Maximum=0.9)
   ) {isSetParameters=true, module=SetVariables_#4}
   &InitializeParticle(
    Color=&SelectLinearColorFromArray(
     Array Sampling Mode=::NewEnumerator0
     Direct Array Index=?
     Color Selection Array=DataInterface() {propertyValues {ColorData [<rgba 1 0 0.0007 1>, <rgba 1 0.0156 0 1>]}}
     Random Seed=0
    )
    Lifetime=2.3
    Sprite Size=&Multiply_Vector2D(
     A=&Lerp_Vector2(A=<43.0435 59.2308>, B=<47 71.0769>, Alpha=$Particles.Scale)
     B=&SelectVector2DFromArray(
      Array Sampling Mode=::NewEnumerator0
      Direct Array Index=?
      Vector 2D Selection Array=DataInterface() {propertyValues {FloatData [<-1 1>, <1 1>]}}
      Random Seed=0
     )
    )
    Sprite Size Mode=::NewEnumerator4
    UsePositionOffset=true
    Write Material Random=true
   )
   &SpriteFacingAndAlignment(Sprite Alignment=<0 0 1>, WriteAlignment=true, WriteFacing=true)
   /SetVariables(Particles.Color=&MultiplyLinearColor(
     Linear Color=$Particles.Color
     Scale Factor=&MakeVector4FromVectorAndFloat(
      "Vector (RGB)"=&VectorFromFloat(Value=&FloatFromCurve(
        FloatCurve=DataInterface() {propertyValues {Curve {keys [
            <at 0 0.2 RCIM_Cubic>
            {arriveTangent=0.8956, interpMode=RCIM_Cubic, leaveTangent=-0.4249, tangentMode=RCTM_Break, time=0.502, value=1}
            <at 1 0.2 RCIM_Cubic>
           ]}}}
        CurveIndex=&ReturnNormalizedExecIndex()
       ))
      "Float (Alpha)"=&Lerp_Float(
       A=0.2
       B=2
       Alpha=&FloatFromCurve(
        FloatCurve=DataInterface() {propertyValues {Curve {keys [
            <at 0 0.15 RCIM_Cubic>
            {arriveTangent=3.4824, interpMode=RCIM_Cubic, leaveTangent=-3.548, tangentMode=RCTM_Break, time=0.502, value=1}
            <at 1 0.15 RCIM_Cubic>
           ]}}}
        CurveIndex=&ReturnNormalizedExecIndex()
        Scale Curve=5
       )
      )
     )
    )) {isSetParameters=true, module=SetVariables_#5}
  ]
  particleUpdateScript [
   &ParticleState()
   &ScaleSpriteSize(
    Scale Factor=&Multiply_Vector2D_ByFloat(
     Float=&FloatFromCurve(
      FloatCurve=DataInterface() {propertyValues {Curve {keys [{arriveTangent=-3.0981, interpMode=RCIM_Cubic, leaveTangent=-3.0981, tangentMode=RCTM_User, time=0, value=1.5}, <at 0.4322 1 RCIM_Cubic>]}}}
      CurveIndex=$Emitter.Age
     )
     Vector2D=&MakeVector2D(
      X=&FloatFromCurve(
       FloatCurve=DataInterface() {propertyValues {Curve {
          defaultValue=1
          keys [
           {arriveTangent=9.0725, interpMode=RCIM_Cubic, leaveTangent=9.0725, tangentMode=RCTM_User, time=0, value=0}
           {arriveTangent=-2.5249, interpMode=RCIM_Cubic, leaveTangent=-2.5249, tangentMode=RCTM_User, time=0.0907, value=1}
          ]
         }}}
       Scale Curve=&FloatFromCurve(
        FloatCurve=DataInterface() {propertyValues {Curve {
           defaultValue=1
           keys [
            {interpMode=RCIM_Cubic, tangentMode=RCTM_User, time=0, value=0.65}
            {arriveTangent=1.1863, interpMode=RCIM_Cubic, leaveTangent=-1.2112, tangentMode=RCTM_Break, time=0.5, value=1.2}
            {tangentMode=RCTM_User, time=1, value=0.65}
           ]
          }}}
        CurveIndex=&ReturnNormalizedExecIndex()
        Scale Curve=&RandomRangeFloat(Minimum=1.4, Maximum=1.4)
       )
      )
      Y=&FloatFromCurve(
       FloatCurve=DataInterface() {propertyValues {Curve {
          defaultValue=1
          keys [
           {arriveTangent=9.0725, interpMode=RCIM_Cubic, time=0, value=0}
           {arriveTangent=17.0555, interpMode=RCIM_Cubic, leaveTangent=17.0555, tangentMode=RCTM_User, time=0.0333, value=0.9401}
           {arriveTangent=0.124, interpMode=RCIM_Cubic, leaveTangent=0.124, tangentMode=RCTM_User, time=0.0482, value=1}
           {arriveTangent=0.3152, interpMode=RCIM_Cubic, leaveTangent=0.3152, time=0.8993, value=1}
           {arriveTangent=-0.0405, interpMode=RCIM_Cubic, leaveTangent=-0.0405, tangentMode=RCTM_User, time=1, value=1.3}
          ]
         }}}
       Scale Curve=&FloatFromCurve(
        FloatCurve=DataInterface() {propertyValues {Curve {
           defaultValue=1
           keys [
            {interpMode=RCIM_Cubic, tangentMode=RCTM_User, time=0, value=0.65}
            {arriveTangent=0.8018, interpMode=RCIM_Cubic, leaveTangent=-0.6589, tangentMode=RCTM_Break, time=0.5, value=1}
            {tangentMode=RCTM_User, time=1, value=0.65}
           ]
          }}}
        CurveIndex=&ReturnNormalizedExecIndex()
        Scale Curve=&RandomRangeFloat(Minimum=3.5, Maximum=4.4)
       )
      )
     )
    )
    Scale Sprite Size Mode=::NewEnumerator2
   )
   &SubUVAnimation(
    "Sprite Renderer "=DataInterface() {propertyValues {SpriteRenderer=&self:BigFire_0.NiagaraSpriteRendererProperties_0}}
    Start Frame Offset:NiagaraInt32=&RandomRangeInt(Minimum=0, Maximum=10)
    SubUV Lookup Index:NiagaraFloat=&Multiply_Float(
     A=$Particles.Age
     B=&FloatFromCurve(
      FloatCurve=DataInterface() {propertyValues {Curve {keys [<at 0.2594 1.5>,<at 1 1.3>]}}}
      CurveIndex=$Emitter.Age
      Scale Curve=&RandomRangeFloat(Minimum=0.8, Maximum=1.1)
     )
    )
    UseEndFrame=false
    UseStartFrame=false
   )
   &ShapeLocation(
    MaskSpawnGroup=false
    Offset=<-10 -15 -5>
    Offset Mode=::NewEnumerator0
    Override Seed=false
    Ring / Disc Distribution Mode:ENiagaraSphereDistributionMode=::NewEnumerator1
    Ring Radius=&RandomRangeFloat(Minimum=70, Maximum=72)
    Rotation Angle=0.04
    Rotation Axis=<0 0 1>
    Rotation Mode=::NewEnumerator1
    Shape Primitive=::NewEnumerator4
    U Position=&FloatFromCurve(
     FloatCurve=DataInterface() {propertyValues {Curve {keys [<at 0 -0.5>,<at 1 0>]}}}
     CurveIndex=&Add_Float(
      A=&ReturnNormalizedExecIndex()
      B=&RandomRangeFloat(Minimum=-0.05, Maximum=0.05)
     )
    )
    UseOwnerScale=true
   ) {module=ShapeLocation001}
   &ScaleColor(
    Scale Alpha=&FloatFromCurve(
     FloatCurve=DataInterface() {propertyValues {Curve {keys [<at 0 0>,<at 0.0615 1>,<at 1 1>]}}}
     Scale Curve=&FloatFromCurve(
      FloatCurve=DataInterface() {propertyValues {Curve {keys [{arriveTangent=-1.777, interpMode=RCIM_Cubic, leaveTangent=-1.777, tangentMode=RCTM_User, time=-0.3898, value=1}, <at 1 0.1 RCIM_Cubic>]}}}
      CurveIndex=&DotProduct(
       Vector A=/Game/Effects/Niagara/DynamicInputs/Engine/FN_GetSystemAxis(Axis:ENiagaraOrientationAxis=::YAxis, Flip Axis:NiagaraBool=false)
       Vector B=&ReturnCameraVector(Camera Vectors=::NewEnumerator1, Coordinate Space=::World)
      )
     )
    )
    Scale RGB=&VectorFromFloat(Value=&FloatFromCurve(FloatCurve=DataInterface() {propertyValues {Curve {keys [<at 0.3 1>,<at 0.8 1>]}}}))
    ScaleA=true
    ScaleRGB=true
   )
   &DynamicMaterialParameters(
    Index 0 Param 1=&Add_Float(
     A=&FloatFromCurve(FloatCurve=DataInterface() {propertyValues {Curve {keys [<at 0.8166 0>,<at 1 1>]}}})
     B=&FloatFromCurve(
      FloatCurve=DataInterface() {propertyValues {Curve {keys [<at 1.2 0>,<at 1.5 0>]}}}
      CurveIndex=$Emitter.Age
      Scale Curve=&FloatFromCurve(
       FloatCurve=DataInterface() {propertyValues {Curve {
          defaultValue=0
          keys [
           <at 0 0 RCIM_Cubic>
           {interpMode=RCIM_Cubic, tangentMode=RCTM_User, time=0.9658, value=0}
           {arriveTangent=1.204, interpMode=RCIM_Cubic, leaveTangent=1.204, tangentMode=RCTM_User, time=1, value=1}
          ]
         }}}
       Scale Curve=&RandomRangeFloat(Minimum=0.7, Maximum=1)
      )
     )
    )
    Index 0 Param 2=&Multiply_Float(
     A=&Lerp_Float(A=-0.1, B=-0.3, Alpha=$Particles.Scale)
     B=&FloatFromCurve(
      FloatCurve=DataInterface() {propertyValues {Curve {keys [
          <at 0 0 RCIM_Cubic>
          {arriveTangent=3.4824, interpMode=RCIM_Cubic, leaveTangent=-3.548, tangentMode=RCTM_Break, time=0.502, value=1}
          <at 1 0 RCIM_Cubic>
         ]}}}
      CurveIndex=&ReturnNormalizedExecIndex()
     )
    )
    Index 1 Param 4=&Lerp_Float(
     A=0
     B=1
     Alpha=&FloatFromCurve(
      FloatCurve=DataInterface() {propertyValues {Curve {keys [
          {interpMode=RCIM_Cubic, leaveTangent=1.7312, tangentMode=RCTM_Break, time=0, value=-1}
          {interpMode=RCIM_Cubic, tangentMode=RCTM_Break, time=0.4572, value=-0.4973}
          {arriveTangent=-2.462, interpMode=RCIM_Cubic, tangentMode=RCTM_Break, time=1, value=-1}
         ]}}}
      CurveIndex=&ReturnNormalizedExecIndex()
     )
    )
    Index 1 Vector 1:Vector3f=&MakeVectorFromLinearColorRGB(Color=<rgba 1 0.2736 0 1>)
    Module.Index0Float0Write=true
    Module.Index0Float1Write=true
    Module.Index0Float2Write=true
    Module.Index0Float3Write=true
    Module.Index1Float3Write=true
    Module.Index1VectorWrite=true
    Write Parameter Index 1=::NewEnumerator3
   )
   /SetVariables(
    Particles.ScorchAlignment:Vector3f=<0 1 0>
    Particles.ScorchCol:LinearColor=&MakeLinearColorFromVectorAndFloat(
     "Vector (RGB)"=&MakeVectorFromLinearColorRGB(Color=&MultiplyLinearColor(
       Linear Color=$Particles.Initial.Color
       Scale Factor=&MakeVector4_FromLinearColor(LinearColor=<rgba 8 8 8 1>)
      ))
     "Float (Alpha)"=&FloatFromCurve(FloatCurve=DataInterface() {propertyValues {Curve {
         defaultValue=1
         keys [
          {interpMode=RCIM_Cubic, tangentMode=RCTM_User, time=0, value=0}
          {interpMode=RCIM_Cubic, tangentMode=RCTM_User, time=0.0148, value=0.9958}
          {arriveTangent=-1.0108, interpMode=RCIM_Cubic, leaveTangent=-1.0108, time=0.7691, value=0.8518}
          {tangentMode=RCTM_User, time=1, value=0}
         ]
        }}})
    )
    Particles.ScorchFacing:Vector3f=<0 0 1>
    Particles.ScorchPage:Vector4f=&MakeVector4(
     X=&FloatFromCurve(
      FloatCurve=DataInterface() {propertyValues {Curve {defaultValue=1, keys [{tangentMode=RCTM_User, time=0, value=0}, {tangentMode=RCTM_User, time=1, value=1}]}}}
      Scale Curve=10
     )
     Y=0
     Z=0
     W=0
    )
    Particles.ScorchSize:Vector2f=&Multiply_Vector2D_ByFloat(Float=&RandomRangeFloat(Minimum=1, Maximum=1), Vector2D=<150 150>)
   ) {isSetParameters=true, module=SetVariables_#6}
   &OffsetPosition(Position Offset=<0 0 2.7>, Position Output=1, Position to Offset:NiagaraPosition=$Particles.ShapeLocation001.ShapePosition)
   &OffsetPosition(
    Position Offset=&RandomRangeVector(Minimum=<0 0 2.69>, Maximum=<0 0 2.71>)
    Position Output=1
    Position to Offset:NiagaraPosition=$Particles.ShapeLocation001.ShapePosition
   ) {module=OffsetPosition001}
  ]
  rendererClasses [&NiagaraSpriteRendererProperties]
  renderers [
   {rendererClass=&NiagaraSpriteRendererProperties}
   {rendererClass=&NiagaraSpriteRendererProperties, rendererIndex=1}
   {rendererClass=&NiagaraSpriteRendererProperties, rendererIndex=2}
   {rendererClass=&NiagaraSpriteRendererProperties, rendererIndex=3}
   {rendererClass=&NiagaraSpriteRendererProperties, rendererIndex=4}
   {rendererClass=&NiagaraSpriteRendererProperties, rendererIndex=5}
  ]
 }
 Black_Embers {
  emitterUpdateScript [
   &EmitterState(Life Cycle Mode=::NewEnumerator1, Loop Behavior=::NewEnumerator1, Loop Duration=2, UseLoopDelay=false)
   &SpawnBurst_Instantaneous(Spawn Count=30, Use Spawn Probability=false, UseLoopCountLimit=false)
   &SpawnRate(
    SpawnRate=&FloatFromCurve(
     FloatCurve=DataInterface() {propertyValues {Curve {keys [<at 0 1>,<at 0.1175 0.2>,<at 0.4407 0>]}}}
     CurveIndex=$Emitter.NormalizedLoopAge
     Scale Curve=400
    )
    Use Spawn Probability=false
   )
  ]
  particleSpawnScript [
   &InitializeParticle(
    Color Maximum=<rgba 0.1 0.0455 0.0122 5>
    Color Minimum=<rgba 0 0 0 5>
    Color Mode=::NewEnumerator2
    Lifetime=&RandomRangeFloat(Minimum=1.4, Maximum=2.5)
    Sprite Size Mode=::NewEnumerator3
    Uniform Sprite Size=7
    UsePositionOffset=false
    Write Material Random=true
   )
   /SetVariables(Particles.Placement:NiagaraFloat=&RandomRangeFloat(Minimum=0, Maximum=1)) {isSetParameters=true}
   &ShapeLocation(
    Cylinder Height=200
    Cylinder Radius=70
    MaskSpawnGroup=false
    Offset=<-10 -15 40>
    Offset Mode=::NewEnumerator0
    Override Seed=false
    Shape Primitive=::NewEnumerator1
    Surface Only=false
    UseOwnerScale=true
   )
   &AddVelocity(Velocity=&RandomRangeVector(Minimum=<0 0 0>, Maximum=<0 0 500>)) {enabled=false}
  ]
  particleUpdateScript [
   &ParticleState()
   &Drag(
    Drag=&FloatFromCurve(
     FloatCurve=DataInterface() {propertyValues {Curve {defaultValue=0, keys [{tangentMode=RCTM_User, time=0, value=0}, {tangentMode=RCTM_User, time=0.3, value=1}]}}}
     Scale Curve=10
    )
    Use Linear Drag=true
    Use Rotational Drag=true
   )
   &CurlNoiseForce(
    Bias Noise=false
    Noise Strength=&FloatFromCurve(FloatCurve=DataInterface() {propertyValues {Curve {keys [<at 0 1>,<at 1 0>]}}}, Scale Curve=500)
    Pan Noise=true
    Pan Noise Field=<0 0 1>
   )
   &GravityForce(Gravity=&MakeVector(
     X=0
     Y=0
     Z=&FloatFromCurve(
      FloatCurve=DataInterface() {propertyValues {Curve {defaultValue=1, keys [{tangentMode=RCTM_User, time=0, value=1}, {tangentMode=RCTM_User, time=1, value=0}]}}}
      Scale Curve=450
     )
    ))
   &SolveForcesAndVelocity(Clamp Velocity=false, Limit Acceleration=false)
   &ScaleColor(
    Scale Alpha=&FloatFromCurve(FloatCurve=DataInterface() {propertyValues {Curve {keys [<at 0 0>,<at 0.1105 1>,<at 0.5356 1>,<at 1 0>]}}}, Scale Curve=0.7)
    ScaleA=true
    ScaleRGB=true
   )
  ]
  rendererClasses [&NiagaraSpriteRendererProperties]
  renderers [{rendererClass=&NiagaraSpriteRendererProperties}]
 }
 Cyc {
  emitterUpdateScript [
   &EmitterState(Life Cycle Mode=::NewEnumerator1, Loop Behavior=::NewEnumerator1, UseLoopDelay=false)
   &SpawnBurst_Instantaneous(Use Spawn Probability=false, UseLoopCountLimit=false)
   /SetVariables(Emitter.SceneDepthThreshold:NiagaraFloat=&Add_Float(
     A=&DistanceBetweenPositions(
      A=&Multiply_Vector(
       A=&ReturnCameraPosition(Coordinate Space:ENiagaraCoordinateSpace=::World, Camera Query:NiagaraDataInterfaceCamera=DataInterface())
       B=<1 1 0>
      )
      B=&Multiply_Vector(A=Linked() {linkedVariable {name=Engine.Owner.Position, type=NiagaraPosition}}, B=<1 1 0>)
     )
     B=30
    )) {isSetParameters=true}
  ]
  particleSpawnScript [&InitializeParticle(
    Color=<rgba 1 0.0057 0 3>
    Lifetime=2.2
    Mesh Scale Mode=::NewEnumerator3
    Mesh Uniform Scale=2
    Position Offset=<0 -10 110>
    UsePositionOffset=true
    Write Material Random=true
   )]
  particleUpdateScript [
   &ParticleState()
   &ScaleColor(
    Scale Alpha=&FloatFromCurve(FloatCurve=DataInterface() {propertyValues {Curve {
        defaultValue=1
        keys [
         {interpMode=RCIM_Cubic, tangentMode=RCTM_User, time=0, value=0}
         {interpMode=RCIM_Cubic, tangentMode=RCTM_User, time=0.0607, value=1}
         {arriveTangent=-0.1458, interpMode=RCIM_Cubic, leaveTangent=-0.1458, tangentMode=RCTM_User, time=0.7334, value=1}
         {tangentMode=RCTM_User, time=1, value=0}
        ]
       }}})
    ScaleA=true
    ScaleRGB=true
   )
  ]
  rendererClasses [&NiagaraMeshRendererProperties]
  renderers [{rendererClass=&NiagaraMeshRendererProperties, rendererIndex=0}]
 }
 Embers001 {
  emitterUpdateScript [
   &EmitterState(Life Cycle Mode=::NewEnumerator1, Loop Behavior=::NewEnumerator1, Loop Duration=1.5, UseLoopDelay=false)
   &SpawnRate(
    SpawnRate=&FloatFromCurve(FloatCurve=DataInterface() {propertyValues {Curve {keys [<at 0 1>,<at 0.12 0>]}}}, CurveIndex=$Emitter.NormalizedLoopAge, Scale Curve=500)
    Use Spawn Probability=false
   )
  ]
  particleSpawnScript [
   &InitializeParticle(
    Color Maximum=<rgba 20 1.4768 0 1>
    Color Minimum=<rgba 20 0 0.199 1>
    Color Mode=::NewEnumerator2
    Lifetime=2.4
    Mass Mode=::NewEnumerator1
    Position=&SimulationPosition()
    Position Mode=::NewEnumerator1
    Position Offset=<0 0 10>
    Sprite Rotation Mode=::NewEnumerator2
    Sprite Size=<3 3>
    Sprite Size Mode=::NewEnumerator4
    UsePositionOffset=false
    Write Material Random=true
   ) {module=InitializeParticle001}
   &ShapeLocation(
    MaskSpawnGroup=false
    Offset Mode=::NewEnumerator0
    Override Seed=false
    Ring Radius=50
    Shape Primitive=::NewEnumerator4
    UseOwnerScale=true
   )
   &AddVelocity(
    Velocity=&Multiply_Vector(A=&RandomRangeVector(Minimum=<10 10 10>, Maximum=<-10 -10 50>), B=<5 5 5>)
    Velocity Speed Scale=&FloatFromCurve(
     FloatCurve=DataInterface() {propertyValues {Curve {keys [<at 0 1 RCIM_Constant>,<at 0.2 2 RCIM_Constant>,<at 0.4 3 RCIM_Constant>,<at 0.6 4 RCIM_Constant>,<at 0.8 5 RCIM_Constant>]}}}
     CurveIndex=&RemapRange(
      Input Value To Remap=&MakeFloatFromInt(Int=$Emitter.SpawnGroup)
      Input Min=0
      Input Max=4
      Output Min=0
      Output Max=1
     )
     Scale Curve=3
    )
   )
  ]
  particleUpdateScript [
   &ParticleState()
   &ScaleSpriteSize(
    Scale Sprite Size Mode=::NewEnumerator0
    Uniform Scale Factor=&MaskFloat(
     Mask Source=0
     Mask Parameter=?
     Mask=$Particles.VisibilityTag
     Masked Float=&FloatFromCurve(FloatCurve=DataInterface() {propertyValues {Curve {keys [
          {arriveTangent=0.178, interpMode=RCIM_Cubic, leaveTangent=0.178, tangentMode=RCTM_User, time=0, value=0.5}
          {arriveTangent=-2.5249, interpMode=RCIM_Cubic, leaveTangent=-2.5249, tangentMode=RCTM_User, time=0.1408, value=0.9847}
          {arriveTangent=-0.0405, interpMode=RCIM_Cubic, leaveTangent=-0.0405, tangentMode=RCTM_User, time=1, value=0}
         ]}}})
     Passthrough Float=&FloatFromCurve(FloatCurve=DataInterface() {propertyValues {Curve {
         defaultValue=1
         keys [
          {interpMode=RCIM_Cubic, tangentMode=RCTM_User, time=0, value=0.5}
          {interpMode=RCIM_Cubic, tangentMode=RCTM_User, time=0.2, value=1}
          {tangentMode=RCTM_User, time=1, value=0}
         ]
        }}})
    )
   )
   &Drag(
    Drag=&FloatFromCurve(
     FloatCurve=DataInterface() {propertyValues {Curve {defaultValue=1, keys [{tangentMode=RCTM_User, time=0, value=0}, {tangentMode=RCTM_User, time=0.2334, value=1}]}}}
     Scale Curve=30
    )
    Use Linear Drag=true
    Use Rotational Drag=false
   )
   /Niagara/Modules/Update/Forces/CurlNoiseForce(Noise Frequency=0.1, Noise Strength=500, Pan Noise=false, Pan Noise Field=<0 0 0.5>)
   &SolveForcesAndVelocity(Clamp Velocity=false, Limit Acceleration=false, Speed Limit=350)
   &ScaleSpriteSizeBySpeed(
    Max Scale Factor=/Niagara/DynamicInputs/Helpers/MaskVector2d(
     Mask Source:NiagaraInt32=0
     Mask Parameter:NiagaraInt32=?
     Mask:NiagaraInt32=$Particles.VisibilityTag
     Masked Vector2D:Vector2f=<0.5 6>
     Passthrough Vector2D:Vector2f=<0.5 0.5>
    )
    Min Scale Factor=<0.5 0.5>
    Sample Scale Factor By Curve=true
    Scale Factor Curve=DataInterface() {propertyValues {Curve {keys [
        <at -0.0077 0.0765 RCIM_Cubic>
        {arriveTangent=1.2463, interpMode=RCIM_Cubic, leaveTangent=1.2463, tangentMode=RCTM_User, time=0.9985, value=0.9549}
       ]}}}
   )
   &ScaleColor(
    Scale Alpha=&FloatFromCurve(FloatCurve=DataInterface() {propertyValues {Curve {keys [<at 0 1>,<at 1 0>]}}})
    Scale RGB=&MakeVectorFromLinearColorRGB(Color=&ColorFromCurve(ColorCurve=DataInterface() {propertyValues {
        AlphaCurve {keys [<at 0 1 RCIM_Cubic>,<at 0.7107 1>]}
        BlueCurve {keys [<at 0 14.6701 RCIM_Cubic>,<at 0.2499 1.5104>,<at 0.7114 0.1753 RCIM_Cubic>]}
        GreenCurve {keys [<at 0 36.7481 RCIM_Cubic>,<at 0.2499 5.293>,<at 0.7114 0.3243 RCIM_Cubic>]}
        RedCurve {keys [<at 0 50 RCIM_Cubic>,<at 0.2499 10>,<at 0.7114 1 RCIM_Cubic>]}
       }}))
    ScaleA=true
    ScaleRGB=true
   ) {enabled=false}
  ]
  rendererClasses [&NiagaraSpriteRendererProperties]
  renderers [{rendererClass=&NiagaraSpriteRendererProperties}]
 }
 Glows {
  emitterUpdateScript [
   &EmitterState(Life Cycle Mode=::NewEnumerator1, Loop Behavior=::NewEnumerator1, Loop Duration=6, UseLoopDelay=false)
   &SpawnBurst_Instantaneous(Spawn Count=0, Use Spawn Probability=false, UseLoopCountLimit=false)
   &SpawnBurst_Instantaneous(Spawn Group=1, Use Spawn Probability=false, UseLoopCountLimit=false) {module=SpawnBurst_Instantaneous001}
   &SpawnBurst_Instantaneous(Spawn Group=2, Use Spawn Probability=false, UseLoopCountLimit=false) {module=SpawnBurst_Instantaneous002}
   &SpawnBurst_Instantaneous(Spawn Group=3, Use Spawn Probability=false, UseLoopCountLimit=false) {module=SpawnBurst_Instantaneous003}
  ]
  particleSpawnScript [
   &InitializeParticle(
    Color=<rgba 1 0.0354 0.0014 1>
    Lifetime=2.3
    Position Offset=/Niagara/DynamicInputs/TypeConversions/MakeCustomPositionFromBool(
     Bool:NiagaraBool=&IsPlatformSetActive(New Platform Set=DataInterface() {propertyValues {Platforms {qualityLevelMask=-2, deviceProfileStates [{profileName=Mobile, qualityLevelMask=0, setQualityLevelMask=4294967295}]}}})
     True Position:NiagaraPosition=&SelectVectorFromArray(
      Direct Array Index=$Emitter.SpawnGroup
      Vector Selection Array=DataInterface() {propertyValues {FloatData [<0 0 5>, <0 -30 30>, <0 0 50>, <0 -20 0.1>]}}
     )
     False Position:NiagaraPosition=&SelectVectorFromArray(
      Direct Array Index=$Emitter.SpawnGroup
      Vector Selection Array=DataInterface() {propertyValues {FloatData [<0 0 5>, <0 -30 30>, <0 0 100>, <0 -20 2>]}}
     )
    )
    Sprite Size Mode=::NewEnumerator3
    Uniform Sprite Size=&MakeCustomFloatFromBool(
     Bool=&IsPlatformSetActive(New Platform Set=DataInterface() {propertyValues {Platforms {qualityLevelMask=-2, deviceProfileStates [{profileName=Mobile, qualityLevelMask=0, setQualityLevelMask=4294967295}]}}})
     True Float=400
     False Float=&SelectFloatFromArray(Direct Array Index=$Emitter.SpawnGroup, Float Selection Array=DataInterface() {propertyValues {FloatData [0,400,400,400]}})
    )
    UsePositionOffset=true
    Write Material Random=true
   )
   /SetVariables(Particles.VisibilityTag=&SelectIntFromArray(Direct Array Index=$Emitter.SpawnGroup, Int Selection Array=DataInterface() {propertyValues {IntData [0,0,1,2]}})) {isSetParameters=true}
   /SetVariables(Particles.GroundFacing:Vector3f=<0 0 1>) {isSetParameters=true}
   /SetVariables(Particles.Color=&MultiplyLinearColor(
     Linear Color=$Particles.Color
     Scale Factor=&MakeVector4(
      X=1
      Y=1
      Z=1
      W=&SelectFloatFromArray(Direct Array Index=$Emitter.SpawnGroup, Float Selection Array=DataInterface() {propertyValues {FloatData [1,1,1,0.5]}})
     )
    )) {isSetParameters=true}
   &CameraOffset(Camera Offset Amount=&MakeCustomFloatFromBool(
     Bool=&IsPlatformSetActive(New Platform Set=DataInterface() {propertyValues {Platforms {qualityLevelMask=-2, deviceProfileStates [{profileName=Mobile, qualityLevelMask=0, setQualityLevelMask=4294967295}]}}})
     True Float=0
     False Float=&SelectFloatFromArray(Direct Array Index=$Emitter.SpawnGroup, Float Selection Array=DataInterface() {propertyValues {FloatData [0,0,-40,0]}})
    ))
  ]
  particleUpdateScript [
   &ParticleState()
   &ScaleColor(
    Scale Alpha=&FloatFromCurve(FloatCurve=DataInterface() {propertyValues {Curve {
        defaultValue=1
        keys [
         {arriveTangent=9.0725, interpMode=RCIM_Cubic, leaveTangent=9.0725, tangentMode=RCTM_User, time=0.0059, value=0}
         {arriveTangent=0.0547, interpMode=RCIM_Cubic, leaveTangent=0.0547, tangentMode=RCTM_User, time=0.0491, value=0.9847}
         {arriveTangent=-0.0334, interpMode=RCIM_Cubic, leaveTangent=-0.0334, tangentMode=RCTM_User, time=0.8677, value=1}
         {arriveTangent=-0.0405, interpMode=RCIM_Cubic, leaveTangent=-0.0405, tangentMode=RCTM_User, time=1, value=0}
        ]
       }}})
    Scale RGB=&VectorFromFloat(Value=&FloatFromCurve(FloatCurve=DataInterface() {propertyValues {Curve {keys [<at 0.0401 2>,<at 0.0712 1>,<at 1 0>]}}}))
    ScaleA=true
    ScaleRGB=true
   )
  ]
  rendererClasses [&NiagaraLightRendererProperties, &NiagaraSpriteRendererProperties]
  renderers [
   {rendererClass=&NiagaraLightRendererProperties, rendererIndex=0}
   {rendererClass=&NiagaraSpriteRendererProperties, rendererIndex=1}
   {rendererClass=&NiagaraSpriteRendererProperties, rendererIndex=2}
  ]
 }
 SmallFire {
  emitterUpdateScript [
   &EmitterState(Life Cycle Mode=::NewEnumerator1, Loop Behavior=::NewEnumerator1, Loop Duration=4, UseLoopDelay=false)
   &SpawnBurst_Instantaneous(Spawn Count=18, Use Spawn Probability=false, UseLoopCountLimit=false)
   &SpawnRate(
    SpawnRate=&FloatFromCurve(FloatCurve=DataInterface() {propertyValues {Curve {keys [<at 0.648 1>,<at 1 0>]}}}, CurveIndex=$Emitter.NormalizedLoopAge, Scale Curve=15)
    Use Spawn Probability=false
   ) {enabled=false}
  ]
  particleSpawnScript [
   /SetVariables(Particles.VisibilityTag=&RandomRangeInt(Minimum=0, Maximum=0)) {isSetParameters=true, module=SetVariables_#2}
   /SetVariables(Particles.Color Array=DataInterface() {propertyValues {GpuSyncMode=SyncCpuToGpu, MaxElements=0, ColorData [<rgba 1 0 0.0007 1>, <rgba 1 0.0896 0 1>, <rgba 0 0 0 0>]}}) {isSetParameters=true, module=SetVariables_#3}
   /SetVariables(
    Particles.Placement:NiagaraFloat=&RandomRangeFloat(Minimum=0, Maximum=1)
    Particles.Scale:NiagaraFloat=&RandomRangeFloat(Minimum=0.05, Maximum=0.1)
   ) {isSetParameters=true, module=SetVariables_#4}
   &InitializeParticle(
    Color=&SelectLinearColorFromArray(
     Array Sampling Mode=::NewEnumerator0
     Direct Array Index=?
     Color Selection Array=DataInterface() {propertyValues {ColorData [<rgba 1 0 0.0007 1>, <rgba 1 0.0156 0 1>]}}
     Random Seed=0
    )
    Lifetime=&RandomRangeFloat(Minimum=2.2, Maximum=2.4)
    Sprite Size=&Multiply_Vector2D(
     A=&Lerp_Vector2(A=<43.0435 59.2308>, B=<47 71.0769>, Alpha=$Particles.Scale)
     B=&SelectVector2DFromArray(
      Array Sampling Mode=::NewEnumerator0
      Direct Array Index=?
      Vector 2D Selection Array=DataInterface() {propertyValues {FloatData [<-1 1>, <1 1>]}}
      Random Seed=0
     )
    )
    Sprite Size Mode=::NewEnumerator4
    UsePositionOffset=true
    Write Material Random=true
   )
   &SpriteFacingAndAlignment(Sprite Alignment=<0 0 1>, WriteAlignment=true, WriteFacing=true)
   /SetVariables(Particles.Color=&LerpLinearColors(
     StartColor=<rgba 0.8125 0.0884 0 1>
     EndColor=<rgba 0.1927 0.0025 0.0023 1>
     LerpFactor=&FloatFromCurve(
      FloatCurve=DataInterface() {propertyValues {Curve {
         defaultValue=1
         keys [
          {arriveTangent=1.8735, interpMode=RCIM_Cubic, leaveTangent=1.8735, tangentMode=RCTM_User, time=0.0898, value=1}
          {arriveTangent=1.1043, interpMode=RCIM_Cubic, leaveTangent=-0.1142, tangentMode=RCTM_Break, time=0.5, value=0}
          {arriveTangent=-0.4864, tangentMode=RCTM_User, time=0.9164, value=1}
         ]
        }}}
      CurveIndex=&ReturnNormalizedExecIndex()
     )
    )) {isSetParameters=true}
   &CameraOffset(Camera Offset Amount=&FloatFromCurve(
     FloatCurve=DataInterface() {propertyValues {Curve {keys [
         {arriveTangent=0.0498, interpMode=RCIM_Cubic, leaveTangent=0.0498, tangentMode=RCTM_User, time=0.0997, value=1}
         <at 0.3205 -1 RCIM_Cubic>
         <at 0.5347 1 RCIM_Cubic>
        ]}}}
     CurveIndex=&ReturnNormalizedExecIndex()
     Scale Curve=30
    ))
  ]
  particleUpdateScript [
   &ParticleState()
   &ScaleSpriteSize(
    Scale Factor=&Multiply_Vector2D_ByFloat(
     Float=&FloatFromCurve(
      FloatCurve=DataInterface() {propertyValues {Curve {keys [{arriveTangent=-3.0981, interpMode=RCIM_Cubic, leaveTangent=-3.0981, tangentMode=RCTM_User, time=0, value=1.5}, <at 0.4322 1 RCIM_Cubic>]}}}
      CurveIndex=$Emitter.Age
     )
     Vector2D=&MakeVector2D(
      X=&FloatFromCurve(
       FloatCurve=DataInterface() {propertyValues {Curve {
          defaultValue=1
          keys [
           {arriveTangent=9.0725, interpMode=RCIM_Cubic, leaveTangent=9.0725, tangentMode=RCTM_User, time=0, value=0}
           {arriveTangent=-2.5249, interpMode=RCIM_Cubic, leaveTangent=-2.5249, tangentMode=RCTM_User, time=0.0263, value=1}
          ]
         }}}
       Scale Curve=&FloatFromCurve(
        FloatCurve=DataInterface() {propertyValues {Curve {
           defaultValue=1
           keys [
            {interpMode=RCIM_Cubic, tangentMode=RCTM_User, time=0, value=0.5}
            {interpMode=RCIM_Cubic, tangentMode=RCTM_User, time=0.5, value=2}
            {tangentMode=RCTM_User, time=1, value=0.5}
           ]
          }}}
        CurveIndex=&ReturnNormalizedExecIndex()
        Scale Curve=&RandomRangeFloat(Minimum=0.8, Maximum=1.2)
       )
      )
      Y=&FloatFromCurve(
       FloatCurve=DataInterface() {propertyValues {Curve {
          defaultValue=1
          keys [
           {arriveTangent=9.0725, interpMode=RCIM_Cubic, leaveTangent=9.0725, tangentMode=RCTM_User, time=0, value=0}
           {arriveTangent=-0.2297, interpMode=RCIM_Cubic, leaveTangent=-0.2297, tangentMode=RCTM_User, time=0.0421, value=0.9847}
           {arriveTangent=-0.1928, interpMode=RCIM_Cubic, leaveTangent=-0.1928, time=0.3502, value=1}
           {arriveTangent=-0.0405, interpMode=RCIM_Cubic, leaveTangent=-0.0405, tangentMode=RCTM_User, time=1, value=0.8}
          ]
         }}}
       Scale Curve=&FloatFromCurve(
        FloatCurve=DataInterface() {propertyValues {Curve {
           defaultValue=1
           keys [
            {interpMode=RCIM_Cubic, tangentMode=RCTM_User, time=0, value=1}
            {arriveTangent=0.8018, interpMode=RCIM_Cubic, leaveTangent=-0.6589, tangentMode=RCTM_Break, time=0.5, value=2}
            {tangentMode=RCTM_User, time=1, value=0.5}
           ]
          }}}
        CurveIndex=&ReturnNormalizedExecIndex()
        Scale Curve=&RandomRangeFloat(Minimum=0.5, Maximum=0.6)
       )
      )
     )
    )
    Scale Sprite Size Mode=::NewEnumerator2
   )
   &SubUVAnimation(
    "Sprite Renderer "=DataInterface() {propertyValues {SpriteRenderer=&self:SmallFire_0.NiagaraSpriteRendererProperties_0}}
    Start Frame Offset:NiagaraInt32=&RandomRangeInt(Minimum=0, Maximum=10)
    SubUV Lookup Index:NiagaraFloat=&Multiply_Float(
     A=$Particles.Age
     B=&FloatFromCurve(
      FloatCurve=DataInterface() {propertyValues {Curve {keys [<at 0.2594 1.5>,<at 1 1.3>]}}}
      CurveIndex=$Emitter.Age
      Scale Curve=&RandomRangeFloat(Minimum=0.8, Maximum=1.1)
     )
    )
    UseEndFrame=false
    UseStartFrame=false
   )
   &ShapeLocation(
    MaskSpawnGroup=false
    Offset=<-10 -15 -2>
    Offset Mode=::NewEnumerator0
    Override Seed=false
    Ring / Disc Distribution Mode:ENiagaraSphereDistributionMode=::NewEnumerator1
    Ring Radius=&RandomRangeFloat(Minimum=70, Maximum=72)
    Rotation Angle=0.04
    Rotation Axis=<0 0 1>
    Rotation Mode=::NewEnumerator1
    Shape Primitive=::NewEnumerator4
    U Position=&FloatFromCurve(
     FloatCurve=DataInterface() {propertyValues {Curve {keys [<at 0 -0.6>,<at 1 0.05>]}}}
     CurveIndex=&Add_Float(
      A=&ReturnNormalizedExecIndex()
      B=&RandomRangeFloat(Minimum=-0.02, Maximum=0.02)
     )
    )
    UseOwnerScale=true
   ) {module=ShapeLocation001}
   &ScaleColor(
    Scale Alpha=&FloatFromCurve(
     FloatCurve=DataInterface() {propertyValues {Curve {keys [<at 0 1>,<at 0.4972 1>,<at 1 1>]}}}
     CurveIndex=&ReturnNormalizedExecIndex()
     Scale Curve=&FN_FrontEndFloatModifier(
      Gameplay=&Multiply_Float(
       A=&FNTimeOfDayScalar(Night Scale=10, Morning Scale=10, Day Scale=15, Evening Scale=10)
       B=&FloatFromCurve(FloatCurve=DataInterface() {propertyValues {Curve {keys [<at 0 0>,<at 0.06 1>]}}})
      )
      Front End=5
     )
    )
    Scale RGB=&VectorFromFloat(Value=&FloatFromCurve(FloatCurve=DataInterface() {propertyValues {Curve {keys [<at 0.3 1>,<at 0.8 1>]}}}))
    ScaleA=true
    ScaleRGB=true
   )
   &DynamicMaterialParameters(
    Index 0 Param 1=&Add_Float(
     A=&FloatFromCurve(FloatCurve=DataInterface() {propertyValues {Curve {keys [<at 0.9377 0>,<at 1 1>]}}})
     B=&FloatFromCurve(
      FloatCurve=DataInterface() {propertyValues {Curve {keys [<at 1.2 0>,<at 1.5 0>]}}}
      CurveIndex=$Emitter.Age
      Scale Curve=&FloatFromCurve(
       FloatCurve=DataInterface() {propertyValues {Curve {
          defaultValue=0
          keys [
           {interpMode=RCIM_Cubic, tangentMode=RCTM_User, time=0.9658, value=0.5}
           {arriveTangent=1.204, interpMode=RCIM_Cubic, leaveTangent=1.204, tangentMode=RCTM_User, time=1, value=1}
          ]
         }}}
       Scale Curve=&RandomRangeFloat(Minimum=0, Maximum=0)
      )
     )
    )
    Index 1 Param 4=-1.4829
    Index 1 Vector 1:Vector3f=&MakeVectorFromLinearColorRGB(Color=<rgba 1 1 1 1>)
    Module.Index0Float0Write=true
    Module.Index0Float1Write=true
    Module.Index0Float2Write=true
    Module.Index0Float3Write=true
    Module.Index1Float3Write=true
    Module.Index1VectorWrite=true
    Write Parameter Index 1=::NewEnumerator3
   )
   /SetVariables(
    Particles.ScorchAlignment:Vector3f=<0 1 0>
    Particles.ScorchCol:LinearColor=&LerpLinearColors(
     StartColor=<rgba 4 0.4864 0 1>
     EndColor=<rgba 1 0.0132 0.0119 1>
     LerpFactor=&FloatFromCurve(
      FloatCurve=DataInterface() {propertyValues {Curve {
         defaultValue=1
         keys [
          {arriveTangent=-1.3374, interpMode=RCIM_Cubic, leaveTangent=-1.3374, tangentMode=RCTM_User, time=0.1179, value=1}
          {arriveTangent=0.8018, interpMode=RCIM_Cubic, leaveTangent=-0.6589, tangentMode=RCTM_Break, time=0.5, value=0}
          {arriveTangent=1.0266, tangentMode=RCTM_User, time=0.9057, value=1}
         ]
        }}}
      CurveIndex=&ReturnNormalizedExecIndex()
     )
    )
    Particles.ScorchDyn1:Vector4f=&MakeVector4(X=&RandomRangeFloat(Minimum=0, Maximum=1), Y=1, Z=0, W=1)
    Particles.ScorchFacing:Vector3f=<0 0 1>
    Particles.ScorchPage:Vector4f=&MakeVector4(
     X=&FloatFromCurve(
      FloatCurve=DataInterface() {propertyValues {Curve {defaultValue=1, keys [{tangentMode=RCTM_User, time=0, value=0}, {tangentMode=RCTM_User, time=1, value=1}]}}}
      Scale Curve=10
     )
     Y=0
     Z=0
     W=0
    )
    Particles.ScorchSize:Vector2f=&Multiply_Vector2D_ByFloat(
     Float=&FloatFromCurve(
      FloatCurve=DataInterface() {propertyValues {Curve {
         defaultValue=1
         keys [
          {interpMode=RCIM_Cubic, tangentMode=RCTM_User, time=0, value=1.6015}
          {arriveTangent=0.8018, interpMode=RCIM_Cubic, leaveTangent=-0.6589, tangentMode=RCTM_Break, time=0.5, value=3.3238}
          {tangentMode=RCTM_User, time=1, value=1.5741}
         ]
        }}}
      CurveIndex=&ReturnNormalizedExecIndex()
      Scale Curve=&Multiply_Float(
       A=&RandomRangeFloat(Minimum=0.5, Maximum=0.6)
       B=&FloatFromCurve(FloatCurve=DataInterface() {propertyValues {Curve {keys [<at 0 0>,<at 0.0313 1.4>,<at 0.0527 1>,<at 0.929 1>,<at 0.9518 1.3199>,<at 0.9669 0>]}}})
      )
     )
     Vector2D=<50 50>
    )
   ) {isSetParameters=true, module=SetVariables_#6}
   &OffsetPosition(Position Offset=<0 0 2.7>, Position Output=1, Position to Offset:NiagaraPosition=$Particles.ShapeLocation001.ShapePosition)
   &OffsetPosition(
    Position Offset=&RandomRangeVector(Minimum=<0 0 1.2>, Maximum=<0 0 1.1>)
    Position Output=1
    Position to Offset:NiagaraPosition=$Particles.ShapeLocation001.ShapePosition
   ) {module=OffsetPosition001}
  ]
  rendererClasses [&NiagaraSpriteRendererProperties]
  renderers [
   {rendererClass=&NiagaraSpriteRendererProperties}
   {rendererClass=&NiagaraSpriteRendererProperties, rendererIndex=1}
   {rendererClass=&NiagaraSpriteRendererProperties, rendererIndex=2}
   {rendererClass=&NiagaraSpriteRendererProperties, rendererIndex=3}
   {rendererClass=&NiagaraSpriteRendererProperties, rendererIndex=4}
   {rendererClass=&NiagaraSpriteRendererProperties, rendererIndex=5}
  ]
 }
 Spreading_fire {
  emitterUpdateScript [
   &EmitterState(Life Cycle Mode=::NewEnumerator1, Loop Behavior=::NewEnumerator1, Loop Duration=1.5, UseLoopDelay=true)
   &SpawnBurst_Instantaneous(Use Spawn Probability=false, UseLoopCountLimit=false)
  ]
  particleSpawnScript [
   &InitializeParticle(
    Color=<rgba 0.25 0.0037 0.0276 0.1>
    Lifetime=2.2
    Mass Mode=::NewEnumerator1
    Position Offset=<0 20 2>
    Sprite Rotation Mode=::NewEnumerator0
    Sprite Size Mode=::NewEnumerator3
    UsePositionOffset=true
    Write Material Random=true
   ) {module=InitializeParticle001}
   &ShapeLocation(
    MaskSpawnGroup=false
    Non Uniform Scale=&MakeVector(
     X=1
     Y=&Add_Float(
      A=&Multiply_Float(A=2, B=&MakeFloatFromInt(Int=$Particles.SideIndex))
      B=-1
     )
     Z=1
    )
    Override Seed=false
    Radius Position:NiagaraFloat=$System.Ring of fire Radius
    Ring / Disc Distribution Mode:ENiagaraSphereDistributionMode=::NewEnumerator1
    Ring Radius=50
    Rotation Mode=::NewEnumerator2
    Shape Primitive=::NewEnumerator4
    U Position=&ModuloFloat(A=&Multiply_Float(A=$Particles.MaterialRandom, B=1), B=0.3)
    UseOwnerScale=true
    Yaw / Pitch / Roll=<0.6889 0 0>
   ) {enabled=false}
   &SpriteFacingAndAlignment(Sprite Facing=<0 0 1>, WriteAlignment=true, WriteFacing=true)
  ]
  particleUpdateScript [
   &ParticleState()
   &ScaleSpriteSize(
    Scale Factor=&RandomRangeVector2D(Minimum=<25 25>, Maximum=<35 35>)
    Scale Sprite Size Mode=::NewEnumerator2
   )
   &ScaleColor(
    Scale Alpha=&FloatFromCurve(
     FloatCurve=DataInterface() {propertyValues {Curve {
        defaultValue=1
        keys [
         {interpMode=RCIM_Cubic, tangentMode=RCTM_User, time=0, value=0}
         {interpMode=RCIM_Cubic, tangentMode=RCTM_User, time=0.0461, value=0.9958}
         {arriveTangent=-0.2207, interpMode=RCIM_Cubic, leaveTangent=-0.2207, tangentMode=RCTM_User, time=0.8384, value=1}
         {tangentMode=RCTM_User, time=1, value=0}
        ]
       }}}
     Scale Curve=&Multiply_Float(
      A=1
      B=&FN_FrontEndFloatModifier(
       Gameplay=&FNTimeOfDayScalar(Night Scale=10, Morning Scale=10, Day Scale=15, Evening Scale=10)
       Front End=5
      )
     )
    )
    Scale RGB=&MakeVectorFromLinearColorRGB(Color=<rgba 5 1.186 0.5217 1>)
    ScaleA=true
    ScaleRGB=true
   )
   &DynamicMaterialParameters(
    Index 0 Param 1=&FloatFromCurve(
     FloatCurve=DataInterface() {propertyValues {Curve {defaultValue=1, keys [{tangentMode=RCTM_User, time=0, value=0}, {tangentMode=RCTM_User, time=1, value=1}]}}}
     Scale Curve=5.5
    )
    Module.Index0Float0Write=true
    Module.Index0Float1Write=true
    Module.Index0Float2Write=true
    Module.Index0Float3Write=true
   )
  ]
  rendererClasses [&NiagaraSpriteRendererProperties]
  renderers [{rendererClass=&NiagaraSpriteRendererProperties}]
 }
}
summary {
 systemName=NS_Emote_TrustFan_Flames
 emitters [
  {emitterName=BigFire, rendererClasses [&NiagaraSpriteRendererProperties]}
  {emitterName=Black_Embers, rendererClasses [&NiagaraSpriteRendererProperties]}
  {emitterName=Glows, rendererClasses [&NiagaraLightRendererProperties, &NiagaraSpriteRendererProperties]}
  {emitterName=Embers001, rendererClasses [&NiagaraSpriteRendererProperties]}
  {emitterName=Spreading_fire, rendererClasses [&NiagaraSpriteRendererProperties]}
  {emitterName=Cyc, rendererClasses [&NiagaraMeshRendererProperties]}
  {emitterName=SmallFire, rendererClasses [&NiagaraSpriteRendererProperties]}
  {emitterName=BackgroundFire, rendererClasses [&NiagaraSpriteRendererProperties]}
 ]
 userVariables [
  User.Anim_Curve_FireSpread {type=NiagaraFloat, defaultValue=1, description=""}
  User.bIsFrontEnd {type=NiagaraBool}
  User.bIsFrontEndPreview {type=NiagaraBool}
  User.Curve {type=NiagaraFloat, defaultValue=1, description=""}
  User.Fire_size {type=NiagaraFloat, defaultValue=1.5, description=""}
  User.Intensity {type=NiagaraFloat, defaultValue=0, description=""}
 ]
}
