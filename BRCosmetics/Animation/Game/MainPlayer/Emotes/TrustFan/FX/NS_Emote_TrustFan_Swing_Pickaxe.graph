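# Niagara system for the pickaxe swing effect of the TrustFan emote, attached to a skeletal mesh.
# An anim-trail emitter draws short-lived, time-of-day-tinted ribbons through four renderers, fading and darkening over life.
# Speed-line sprites spawn along the mesh only while owner velocity exceeds a threshold, and a burst-spawned mesh particle follows a bone.
# SupportingTrails, which picks bones by swing section, and TrailHighlight002 are disabled.
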NiagaraSystem /BRCosmetics/Animation/Game/MainPlayer/Emotes/TrustFan/FX/NS_Emote_TrustFan_Swing_Pickaxe
dependencies {
 usedDataInterfaces [
  &NiagaraDataInterfaceCurve
  &NiagaraDataInterfaceColorCurve
  &NiagaraDataInterfaceArrayFloat
  &NiagaraDataInterfaceVector2DCurve
 ]
 usedDynamicInputs [&MakeFloatFromVector2d]
 usedModules [
  &SystemState
  &self:SystemSpawnScript.SystemScriptSource.SystemScriptGraph.NiagaraNodeAssignment_0.SetVariables_#1
 ]
}
emitters {
 FN_AnimTrail001 {
  emitterSpawnScript [&FN_RibbonIDOffsetUtility()]
  emitterUpdateScript [
   &EmitterState()
   &FN_SpawnAnimTrail(Switch on ENiagara_AnimTrail_Source:ENiagara_AnimTrail_Source=::NewEnumerator1)
  ]
  particleSpawnScript [
   &FN_InitializeAnimTrailBoneIndecies()
   &FN_InitializeAnimTrailBoneIndecies(AnimTrail Selection=1, First Bone or Socket Index=2, Second Bone or Socket Index=3) {enabled=false, module=FN_InitializeAnimTrailBoneIndecies001}
   &FN_InitializeAnimTrail(
    AnimTrailWidth=0.97
    Color=&FNTimeOfDayColor(
     Night Color=<rgba 0.1331 0.1821 0.5677 1>
     Morning Color=<rgba 0.1997 0.2734 0.8522 1>
     Day Color=<rgba 0.2623 0.4233 0.7379 1>
     Evening Color=<rgba 0.2623 0.4233 0.7379 1>
    )
    Lifetime=0.06
    SkeletalMesh=$System.Skeletal Mesh
   )
   &AddVelocity(Velocity=<0 400 0>) {enabled=false}
  ]
  particleUpdateScript [
   &SolveForcesAndVelocity(Clamp Velocity=false, Limit Acceleration=false)
   &FN_UpdateAnimTrail(SkeletalMesh=$System.Skeletal Mesh) {module=FN_UpdateAnimTrail001}
   &ScaleColor(
    Scale Alpha=&FloatFromCurve(FloatCurve=DataInterface() {propertyValues {Curve {keys [{interpMode=RCIM_Cubic, tangentMode=RCTM_User, time=0, value=1}, {interpMode=RCIM_Cubic, tangentMode=RCTM_User, time=1, value=0}]}}})
    Scale RGB=&MakeVectorFromLinearColorRGB(Color=&ColorFromCurve(ColorCurve=DataInterface() {propertyValues {
        AlphaCurve {keys [
          {arriveTangent=-1.791, interpMode=RCIM_Cubic, leaveTangent=-1.791, time=0.4417, value=1}
          {arriveTangent=-1.791, interpMode=RCIM_Cubic, leaveTangent=-1.791, time=1, value=0}
         ]}
        BlueCurve {keys [<at 0 1>, {interpMode=RCIM_Cubic, leaveTangent=-0.8114, time=0.6162, value=0.6354}]}
        GreenCurve {keys [<at 0 1>, {interpMode=RCIM_Cubic, leaveTangent=-0.8114, time=0.6162, value=0.6354}]}
        RedCurve {keys [<at 0 1>, {interpMode=RCIM_Cubic, leaveTangent=-0.8114, time=0.6162, value=0.6354}]}
       }}))
    ScaleA=true
    ScaleRGB=true
   )
   &ScaleRibbonWidth(Ribbon Width Scale=&FloatFromCurve(FloatCurve=DataInterface() {propertyValues {Curve {keys [
         {arriveTangent=-1.2579, interpMode=RCIM_Cubic, time=0, value=1}
         {arriveTangent=-0.6749, interpMode=RCIM_Cubic, leaveTangent=-0.6749, tangentMode=RCTM_User, time=1, value=1}
        ]}}}))
   /SetVariables(Particles.SecondRibbon:LinearColor=<rgba 0.5464 0.714 0.8281 0.2>) {isSetParameters=true}
   /SetVariables(Particles.SpeedLines:LinearColor=<rgba 0.2105 0.287 0.5677 1>) {isSetParameters=true}
  ]
  rendererClasses [&NiagaraRibbonRendererProperties]
  renderers [
   {rendererClass=&NiagaraRibbonRendererProperties}
   {rendererClass=&NiagaraRibbonRendererProperties, rendererIndex=1}
   {rendererClass=&NiagaraRibbonRendererProperties, rendererIndex=2}
   {rendererClass=&NiagaraRibbonRendererProperties, rendererIndex=3}
  ]
 }
 Minimal {
  emitterUpdateScript [
   &EmitterState()
   &SpawnBurst_Instantaneous(Use Spawn Probability=false, UseLoopCountLimit=false)
  ]
  particleSpawnScript [&InitializeParticle(Color Mode=::NewEnumerator0, Lifetime=5, UsePositionOffset=false, Write Material Random=true)]
  particleUpdateScript [
   &ParticleState()
   &SkeletalMeshLocation(Bone / Socket Index=1, Bone Sampling Mode=::NewEnumerator4, MaskSpawnGroup=false, Skeletal Mesh=$System.Skeletal Mesh)
  ]
  rendererClasses [&NiagaraMeshRendererProperties]
  renderers [{rendererClass=&NiagaraMeshRendererProperties, rendererIndex=0}]
 }
 Speed_Lines {
  emitterUpdateScript [
   &EmitterState(Life Cycle Mode=::NewEnumerator1, Loop Duration=0.95, UseLoopDelay=false)
   &SpawnRate(
    SpawnRate=&MakeCustomFloatFromBool(
     Bool=&SetBoolByFloatComparison(
      A=&Abs_Float(Float:NiagaraFloat=&VectorLength(Vector=$Engine.Owner.Velocity))
      B=3
     )
     True Float=500
     False Float=0
    )
    Use Spawn Probability=false
   ) {module=SpawnRate001}
  ]
  particleSpawnScript [
   &InitializeParticle(
    Color=<rgba 0.3325 0.5089 0.7835 0.5>
    Lifetime Max=0.1
    Lifetime Min=0.06
    Lifetime Mode=::NewEnumerator1
    Position Mode=::NewEnumerator1
    Sprite Size Max=<66.6667 40>
    Sprite Size Min=<16.6667 20>
    Sprite Size Mode=::NewEnumerator2
    UsePositionOffset=false
    Write Material Random=true
   )
   &ApplyInitialForces()
   &AddVelocity(Velocity=&Multiply_Vector(
     A=&MakeVector(
      X=0
      Y=0
      Z=&SelectFloatFromArray(
       Direct Array Index=&MakeIntFromFloat(Float=$User.WorkoutSelect)
       Float Selection Array=DataInterface() {propertyValues {FloatData [100,0,0,0]}}
      )
     )
     B=&VectorFromFloat(Value=&SelectFloatFromArray(
       Array Sampling Mode=::NewEnumerator0
       Direct Array Index=?
       Float Selection Array=DataInterface() {propertyValues {FloatData [-1,1]}}
       Random Seed=0
      ))
    )) {module=AddVelocity001}
   &CameraOffset(Camera Offset Amount=&FN_FrontEndFloatModifier(Gameplay=0, Front End=0))
  ]
  particleUpdateScript [
   &ParticleState()
   &Drag(Drag=5, Use Linear Drag=true, Use Rotational Drag=false)
   &SkeletalMeshLocation(
    Bone Sampling Mode=::NewEnumerator4
    MaskSpawnGroup=false
    Sampled Position Offset=&RandomRangeVector(Minimum=<0 0 0>, Maximum=<60 0 0>)
    Skeletal Mesh=$System.Skeletal Mesh
   )
   &SolveForcesAndVelocity(Clamp Velocity=false, Limit Acceleration=false)
   &AlignSpriteToMeshOrientation(Mesh Orientation Relative Sprite Alignment Vector=<1 0 0>, SetAlignment=true, SetFacing=true)
   &ScaleColor(
    Scale Alpha=&FloatFromCurve(FloatCurve=DataInterface() {propertyValues {Curve {keys [
         {interpMode=RCIM_Cubic, tangentMode=RCTM_User, time=0.1552, value=0}
         {interpMode=RCIM_Cubic, tangentMode=RCTM_User, time=0.5, value=1}
         {tangentMode=RCTM_User, time=1, value=0}
        ]}}})
    ScaleA=true
    ScaleRGB=true
   )
   &ScaleSpriteSize(
    Non-Uniform Curve Sprite Scale=DataInterface() {propertyValues {
      XCurve {keys [{tangentMode=RCTM_User, time=0, value=0.5}, {tangentMode=RCTM_User, time=1, value=1.5}]}
      YCurve {keys [{tangentMode=RCTM_User, time=0, value=1}, {tangentMode=RCTM_User, time=1, value=0.5}]}
     }}
    Scale Sprite Size Mode=::NewEnumerator3
   )
  ]
  rendererClasses [&NiagaraSpriteRendererProperties]
  renderers [{rendererClass=&NiagaraSpriteRendererProperties}]
 }
 SupportingTrails {
  bEnabled=false
  emitterSpawnScript [&FN_RibbonIDOffsetUtility()]
  emitterUpdateScript [
   &EmitterState()
   &FN_SpawnAnimTrail(Switch on ENiagara_AnimTrail_Source:ENiagara_AnimTrail_Source=::NewEnumerator1)
  ]
  particleSpawnScript [
   &FN_InitializeAnimTrailBoneIndecies(
    First Bone or Socket Index:NiagaraInt32=&MakeCustomIntFromBool(
     Bool=$User.IsLastSwing
     True INT=3
     False INT=&MergeHelper_IntSwitch(0=5, 1=1, 2=1, 3=3, 4=?, 5=?, Index=$User.SwingSectionIndex, Num=::NewEnumerator2)
    )
    Second Bone or Socket Index:NiagaraInt32=&MakeCustomIntFromBool(
     Bool=$User.IsLastSwing
     True INT=2
     False INT=&MergeHelper_IntSwitch(0=4, 1=0, 2=2, 3=2, 4=?, 5=?, Index=$User.SwingSectionIndex, Num=::NewEnumerator2)
    )
   )
   &FN_InitializeAnimTrail(
    AnimTrailWidth=&MakeCustomFloatFromBool(Bool=$User.IsLastSwing, True Float=1.2, False Float=1)
    Color=<rgba 0 0 0 1>
    Lifetime=0.16
    SkeletalMesh=$System.Skeletal Mesh
   )
  ]
  particleUpdateScript [
   &FN_UpdateAnimTrail(Locked Particle RibbonWidth=1, SkeletalMesh=$System.Skeletal Mesh) {module=FN_UpdateAnimTrail001}
   &ScaleColor(
    Scale Alpha=&FloatFromCurve(FloatCurve=DataInterface() {propertyValues {Curve {keys [
         {interpMode=RCIM_Cubic, tangentMode=RCTM_User, time=0, value=1}
         {arriveTangent=-5.0049, interpMode=RCIM_Cubic, leaveTangent=-5.0049, tangentMode=RCTM_User, time=0.5, value=0}
        ]}}})
    ScaleA=true
    ScaleRGB=false
   )
  ]
  rendererClasses [&NiagaraRibbonRendererProperties]
  renderers [{rendererClass=&NiagaraRibbonRendererProperties}]
 }
 TrailHighlight002 {
  bEnabled=false
  emitterUpdateScript [
   &EmitterState()
   &FN_SpawnAnimTrail(
    AnimTrailWidthMode=::FromSecond
    Spawn Mode:ENiagaraSkeletalSpawning=::NewEnumerator0
    Switch on ENiagara_AnimTrail_Source:ENiagara_AnimTrail_Source=::NewEnumerator1
   )
  ]
  particleSpawnScript [
   &FN_InitializeAnimTrailBoneIndecies()
   &FN_InitializeAnimTrail(AnimTrailWidth=1.6, Color=<rgba 1.6575 1.916 5 1>, Lifetime=0.25, SkeletalMesh=$System.Skeletal Mesh)
  ]
  particleUpdateScript [
   &ParticleState()
   &ScaleColor(
    Scale Alpha=&Multiply_Float(
     A=&FloatFromCurve(FloatCurve=DataInterface() {propertyValues {Curve {keys [<at -0.0077 0 RCIM_Cubic>,<at 0.0403 1 RCIM_Cubic>,<at 1 0 RCIM_Cubic>]}}})
     B=1
    )
    Scale RGB=&VectorFromFloat(Value=1)
    ScaleA=true
    ScaleRGB=true
   )
   &ScaleRibbonWidth(Ribbon Width Scale=1.14)
   &FN_UpdateAnimTrail(SkeletalMesh=$System.Skeletal Mesh)
  ]
  rendererClasses [&NiagaraRibbonRendererProperties]
  renderers [{rendererClass=&NiagaraRibbonRendererProperties}]
 }
}
summary {
 systemName=NS_Emote_TrustFan_Swing_Pickaxe
 emitters [
  {bEnabled=false, emitterName=TrailHighlight002, rendererClasses [&NiagaraRibbonRendererProperties]}
  {emitterName=FN_AnimTrail001, rendererClasses [&NiagaraRibbonRendererProperties]}
  {bEnabled=false, emitterName=SupportingTrails, rendererClasses [&NiagaraRibbonRendererProperties]}
  {emitterName=Speed_Lines, rendererClasses [&NiagaraSpriteRendererProperties]}
  {emitterName=Minimal, rendererClasses [&NiagaraMeshRendererProperties]}
 ]
 userVariables [
  User.bIsFrontEnd {type=NiagaraBool}
  User.bIsFrontEndPreview {type=NiagaraBool}
  User.IsLastSwing {type=NiagaraBool}
  User.OwnerLocalVelocity {type=Vector3f, description="", defaultValue {struct=Vector3f, value=<0 0 0>}}
  User.RibbonIDOffset {type=NiagaraInt32}
  User.SkeletalMesh {
   type=NiagaraDataInterfaceSkeletalMesh
   defaultValue {
    struct=NiagaraExt_VariableValue_DataInterface
    value {dataInterface=&self:NiagaraDataInterfaceSkeletalMesh_0}
   }
  }
  User.SwingSectionIndex {type=NiagaraInt32}
  User.WorkoutSelect {type=NiagaraFloat, defaultValue=0, description=""}
 ]
}
